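# Translucent two-sided material for animated butterfly meshes, with glow and opacity scaled by time of day.
# World position offset combines axis rotation, vertex-color-weighted sine wing flaps, looping directional drift and vector noise (omitted on mobile).
# Vertex alpha varies speed, noise, direction and hue; UV1 offsets the drift loop and samples a glow mask.
# Color comes from the butterfly pickaxe texture or a tinted channel; opacity fades near camera.
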Material /Game/Athena/Effects/Materials/Environment/Apollo/M_Apollo_Butterfly
graph {
 properties {
  bUsedWithInstancedStaticMeshes=true
  bUsedWithMeshDeformer=true
  bUsedWithNanite=true
  bUsedWithSkeletalMesh=true
  blendMode=BLEND_Translucent
  pixelDepthOffsetMode=PDOM_Legacy
  twoSided=true
  usedShadingModels=Default Lit
 }
 nodes [
  n1 = VectorParameter() {parameterName=RotationAxis}
  n2 = Time()
  n3 = ObjectPositionWS()
  n4 = StaticSwitchParameter(True=Add(A=@n2, B=ComponentMask(None=@n3) {b=true}), False=@n2) {parameterName=UsewithGrassActor?}
  n5 = ScalarParameter() {defaultValue=1, parameterName=RotationSpeed}
  n6 = LinearInterpolate(A=@n5, B=Multiply(A=@n5) {constB=2}, Alpha=VertexColor().A)
  n7 = ActorPositionWS()
  n8 = ComponentMask(None=TextureCoordinate()) {r=true}
  n9 = Multiply(
   A=Sign(None=Subtract(A=@n8) {constB=0.5})
   B=LinearInterpolate(
    A=ScalarParameter() {defaultValue=-0.389, parameterName=InwardVector}
    B=ScalarParameter() {defaultValue=-1.0138, parameterName=InwardVector_1}
    Alpha=VertexColor().G
   )
  )
  n10 = ScalarParameter() {defaultValue=0.134, parameterName=ForwardVector}
  n11 = AppendVector(A=@n9, B=@n10)
  n12 = AppendVector(A=@n10, B=@n9)
  n13 = ScalarParameter() {defaultValue=0.5823, parameterName=UPVector}
  n14 = ScalarParameter() {defaultValue=1.4967, parameterName=UPVector_1}
  n15 = VertexColor()
  n16 = Transform(None=AppendVector(A=StaticSwitchParameter(True=@n11, False=@n12) {parameterName=SwizzleXY}, B=LinearInterpolate(A=@n13, B=@n14, Alpha=@n15.B))) {transformSourceType=TRANSFORMSOURCE_Local}
  n17 = ScalarParameter() {defaultValue=15, parameterName=WindMovementDistance}
  n18 = Add(A=Time(), B=VertexColor().A)
  n19 = Multiply(A=VertexColor().G, B=ScalarParameter() {defaultValue=0.2, parameterName=WingOffset}) {constB=0.5}
  n20 = Multiply(A=ScalarParameter() {defaultValue=2, parameterName=SpeedWings}, B=LinearInterpolate(Alpha=VertexColor().A) {constA=0.5})
  n21 = Add(A=ScalarParameter() {defaultValue=1, parameterName=WingAnimSpeedMultiplier}, B=LinearInterpolate(Alpha=VertexColor().A) {constA=0.5})
  n22 = Sine(None=Multiply(A=Add(A=@n18, B=@n19), B=Multiply(A=@n20, B=@n21)))
  n23 = ScalarParameter() {defaultValue=0.7853, parameterName=WindMovementBias}
  n24 = Multiply(
   A=Multiply(A=Multiply(A=@n16, B=@n17), B=Add(A=@n22, B=@n23))
   B=Add(A=VertexColor().R, B=ScalarParameter() {defaultValue=0.2286, parameterName=BodyMovement})
  )
  n25 = ScalarParameter() {defaultValue=50, parameterName=CurIntensity}
  n26 = ScalarParameter() {defaultValue=100, parameterName=CurIntensityHigh}
  n27 = VertexColor()
  n28 = WorldPosition()
  n29 = ActorPositionWS()
  n30 = Time()
  n31 = ScalarParameter() {defaultValue=1, parameterName=CurlSpeed}
  n32 = Divide(
   A=Subtract(A=Subtract(A=@n28.XYZ, B=@n29), B=Multiply(A=@n30, B=@n31))
   B=LinearInterpolate(
    A=ScalarParameter() {defaultValue=1000, parameterName=CurlScale}
    B=ScalarParameter() {defaultValue=500, parameterName=CurlScale_Low}
    Alpha=@n27.A
   )
  )
  n33 = Add(A=@n24, B=Multiply(A=LinearInterpolate(A=@n25, B=@n26, Alpha=@n27.A), B=VectorNoise(World Position=@n32) {noiseFunction=VNF_VectorALU}))
  n34 = Multiply(A=ObjectRadius(), B=Constant() {r=1}) {constB=0.5}
  n35 = Time() {bOverride_Period=true, period=3}
  n36 = Time() {bOverride_Period=true, period=4}
  n37 = ComponentMask(None=TextureCoordinate() {coordinateIndex=1}) {g=true}
  n38 = Add(
   A=StaticSwitchParameter(True=Divide(A=@n35) {constB=3}, False=Divide(A=@n36) {constB=4}) {defaultValue=true, parameterName=UseFastSprites}
   B=Multiply(A=Subtract(A=@n37) {constB=0.5}) {constB=-2}
  )
  n39 = Multiply(
   A=LinearInterpolate(A=Multiply(A=@n34) {constB=-1}, B=@n34, Alpha=Frac(None=@n38))
   B=ScalarParameter() {defaultValue=1, parameterName=OverallSpeed}
  )
  n40 = LinearInterpolate(
   A=VectorParameter() {parameterName=OverallDirection, defaultValue=<rgba 1 0 0 1>}.RGB
   B=VectorParameter() {parameterName=OverallDirectionLow, defaultValue=<rgba 1 0 0 1>}.RGB
   Alpha=VertexColor().A
  )
  Add(
   A=RotateAboutAxis(
    NormalizedRotationAxis=Transform(None=@n1.RGB) {transformSourceType=TRANSFORMSOURCE_Local}
    RotationAngle=Multiply(A=@n4, B=@n6)
    PivotPoint=StaticSwitchParameter(True=@n3, False=@n7) {parameterName=UsewithGrassActor?}
    Position=WorldPosition().XYZ
   )
   B=Add(A=ShadingPathSwitch(Default=@n33, Mobile=@n24) {desc=Migrated from FeatureLevelSwitch}, B=Multiply(A=@n39, B=@n40))
  )
  n41 = Multiply(A=ScalarParameter() {defaultValue=1, parameterName=HueShift}, B=VertexColor().A)
  n42 = TextureSample(UVs=TextureCoordinate()) {texture=/Game/Weapons/FORT_Melee/Pickaxe_Butterfly/Textures/T_Pickaxe_Butterfly_D}
  n43 = StaticSwitchParameter(True=Multiply(A=VectorParameter() {parameterName=Color, defaultValue=<rgba 1 1 1 0>}.RGB, B=@n42.G), False=@n42.RGB) {parameterName=UseOneChannelTexandColor?}
  n44 = VectorParameter() {parameterName=GlowColor, defaultValue=<rgba 1 1 1 0>}
  n45 = MaterialFunctionCall(
   Value_Morning=ScalarParameter() {defaultValue=1, parameterName="Value_Morning (S)"}
   Value_Day=ScalarParameter() {defaultValue=1, parameterName="Value_Day (S)"}
   Value_Evening=ScalarParameter() {defaultValue=1, parameterName="Value_Evening (S)"}
   Value_Night=ScalarParameter() {parameterName="Value_Night (S)"}
  ) {materialFunction=&Scalar_Time_of_Day}
  n46 = TextureCoordinate() {coordinateIndex=1}
  Multiply(
   A=Multiply(
    A=MaterialFunctionCall(Hue Shift Percentage=@n41, Texture=@n43) {materialFunction=&HueShift}.Result
    B=Multiply(A=@n44.RGB, B=@n45.Result)
   )
   B=Clamp(None=TextureSample(UVs=@n46) {texture=/Game/Athena/Prototype/Blueprints/Island/T_Butterfly_M}.G)
  )
  n47 = MaterialFunctionCall(Distance=ScalarParameter() {defaultValue=1000, parameterName=NearCamFade}) {materialFunction=&Near_Camera_Fade}
  n48 = ScalarParameter() {defaultValue=1, parameterName=NearCamFadeExp}
  Multiply(A=Multiply(A=ScalarParameter() {defaultValue=1, parameterName=Opacity}, B=Power(Base=@n47.Result, Exp=@n48)), B=@n45.Result)
 ]
}
