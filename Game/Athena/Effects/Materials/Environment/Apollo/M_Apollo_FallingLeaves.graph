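# Translucent lit environment material for Apollo that animates falling leaves as sprites, moving them entirely on the GPU.
# World position offset combines sprite expansion, looping directional travel, sine drift and axis rotation, staggered by UV1, vertex color and object position.
# Sprite size follows the loop phase and shrinks through depth and near-camera fades.
# A rotating SubUV leaf texture supplies opacity and shading, tinted by time-of-day color with vertex-driven hue shift.
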Material /Game/Athena/Effects/Materials/Environment/Apollo/M_Apollo_FallingLeaves
graph {
 properties {
  bUsedWithInstancedStaticMeshes=true
  bUsedWithNanite=true
  blendMode=BLEND_Translucent
  pixelDepthOffsetMode=PDOM_Legacy
  usedShadingModels=Default Lit
 }
 nodes [
  n1 = WorldPosition()
  n2 = Multiply(A=ScalarParameter() {defaultValue=0.25, parameterName=MinSizeMult}, B=VertexColor().G)
  n3 = ScalarParameter() {defaultValue=500, parameterName=maxSize}
  n4 = Time() {bOverride_Period=true, period=3}
  n5 = Divide(A=Time() {bOverride_Period=true, period=4}) {constB=4}
  n6 = Divide(A=Time() {bOverride_Period=true, period=6}) {constB=6}
  n7 = VertexColor()
  n8 = ComponentMask(None=TextureCoordinate() {coordinateIndex=1}) {g=true}
  n9 = Add(
   A=StaticSwitchParameter(True=Divide(A=@n4) {constB=3}, False=LinearInterpolate(A=@n5, B=@n6, Alpha=@n7.G)) {defaultValue=true, parameterName=UseFastSprites}
   B=Multiply(A=Subtract(A=@n8) {constB=0.5}) {constB=-2}
  )
  n10 = Sine(None=Add(A=@n9) {constB=-0.25})
  n11 = Power(Base=LinearInterpolate(Alpha=ConstantBiasScale(None=@n10))) {constExponent=0.6}
  n12 = ScalarParameter() {defaultValue=50, parameterName=FadeDistance}
  n13 = Constant() {r=1}
  n14 = ShadingPathSwitch(Default=QualitySwitch(Default=DepthFade(FadeDistance=@n12), Low=@n13), Mobile=@n13)
  n15 = Multiply(
   A=LinearInterpolate(A=Multiply(A=@n2, B=@n3), B=@n3, Alpha=Multiply(A=@n11, B=@n14))
   B=MaterialFunctionCall(Distance=ScalarParameter() {defaultValue=500, parameterName=NearCamShrinkDist}) {materialFunction=&Near_Camera_Fade}.Result
  )
  n16 = Multiply(A=Time(), B=ScalarParameter() {defaultValue=0.001, parameterName=VarianceSpeed}) {constB=0.05}
  n17 = ComponentMask(None=TextureCoordinate() {coordinateIndex=1}) {r=true}
  n18 = Add(A=Multiply(A=Time(), B=ScalarParameter() {defaultValue=0.4, parameterName=DriftSpeed}), B=Sine(None=Add(A=@n16, B=@n17)) {period=0.17})
  n19 = Sine(None=@n18) {period=0.56}
  n20 = Sine(None=@n18) {period=0.7}
  n21 = Add(A=@n16, B=@n17)
  n22 = Multiply(
   A=AppendVector(A=AppendVector(A=@n19, B=@n20), B=Sine(None=@n18) {period=0.82})
   B=LinearInterpolate(
    A=ScalarParameter() {defaultValue=1, parameterName=MinHorizontalDrift}
    B=ScalarParameter() {defaultValue=1, parameterName=MaxHorizontalDrift}
    Alpha=Sine(None=@n21) {period=0.14}
   )
  )
  n23 = Multiply(A=ObjectRadius(), B=Constant() {r=0.5}) {constB=0.5}
  n24 = Multiply(A=@n23) {constB=-1}
  n25 = Frac(None=@n9)
  n26 = Multiply(
   A=Multiply(A=LinearInterpolate(A=@n24, B=@n23, Alpha=@n25), B=ScalarParameter() {defaultValue=1, parameterName=OverallSpeed})
   B=VectorParameter() {parameterName=OverallDirection, defaultValue=<rgba 1 0 0 1>}.RGB
  )
  n27 = VectorParameter() {parameterName=RotationAxis}
  n28 = ObjectPositionWS()
  n29 = Add(A=Time(), B=ComponentMask(None=@n28) {b=true})
  n30 = ObjectPositionWS()
  n31 = LinearInterpolate(
   A=ScalarParameter() {defaultValue=1, parameterName=RotationSpeed}
   B=ScalarParameter() {defaultValue=1, parameterName=RotationSpeed_Low}
   Alpha=Frac(None=ComponentMask(None=@n30) {b=true})
  )
  Add(
   A=Add(
    A=MaterialFunctionCall!(Center Location=@n1.XYZ, 0-1 UVs, XY scale=@n15, OverridePivot, "Normals (Optional)") {materialFunction=&Sprite}.World position offset
    B=Add(A=@n22, B=@n26)
   )
   B=RotateAboutAxis(
    NormalizedRotationAxis=Transform(None=@n27.RGB) {transformSourceType=TRANSFORMSOURCE_Local}
    RotationAngle=Multiply(A=@n29, B=@n31)
    PivotPoint=@n28
    Position=WorldPosition().XYZ
   )
  )
  n32 = Multiply(A=VertexColor().R, B=ScalarParameter() {defaultValue=-0.2, parameterName=HueShiftAmount}) {constB=-0.2}
  n33 = MaterialFunctionCall(
   Color_Morning=VectorParameter() {parameterName="Color_Morning (V3)", defaultValue=<rgba 1 1 1 0>}.RGB
   Color_Day=VectorParameter() {parameterName="Color_Day (V3)", defaultValue=<rgba 1 1 1 0>}.RGB
   Color_Evening=VectorParameter() {parameterName="Color_Evening (V3)", defaultValue=<rgba 1 1 1 0>}.RGB
   Color_Night=VectorParameter() {parameterName="Color_Night (V3)", defaultValue=<rgba 1 1 1 0>}.RGB
  ) {materialFunction=&Color_Time_of_Day}
  n34 = ScalarParameter() {defaultValue=8, parameterName="U_Images (S)"}
  n35 = ScalarParameter() {defaultValue=4, parameterName="Time Speed (S) Slow"}
  n36 = ScalarParameter() {defaultValue=4, parameterName="Time Speed (S)"}
  n37 = VertexColor()
  n38 = TextureCoordinate()
  n39 = VectorParameter() {parameterName="Rotation Center (V2)", defaultValue=<rgba 0.5 0.5 0 0>}
  n40 = Add(A=Time(), B=Add(A=@n36, B=@n37.A))
  n41 = TextureSampleParameter2D(UVs=MaterialFunctionCall!(
    U_Images=@n34
    V_Images=@n34
    SubUV Texture RGB
    Time Speed=LinearInterpolate(A=@n35, B=@n36, Alpha=@n37.A)
    Manual Time Control
    Time
    UVs=MaterialFunctionCall(UVs=@n38, Rotation Center=@n39.RGB, "Rotation Angle (0-1)"=@n40) {materialFunction=&CustomRotator}.Rotated Values
   ) {materialFunction=/Game/Effects/Fort_Effects/Materials/_Functions/MF_SubUV}.UVs) {parameterName=Texture, texture=/Game/Effects/Fort_Effects/Textures/Leaf/T_Leaf_SubUV}
  Multiply(
   A=ScalarParameter() {defaultValue=1, parameterName=Dark}
   B=Multiply(
    A=MaterialFunctionCall(Hue Shift Percentage=@n32, Texture=@n33.Result) {materialFunction=&HueShift}.Result
    B=@n41.G
   )
  )
  Multiply(A=@n41.A, B=ScalarParameter() {parameterName=TEMPOPACITY})
 ]
}
