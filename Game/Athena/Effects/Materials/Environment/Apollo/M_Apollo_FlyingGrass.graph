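# Translucent two-sided material for Apollo environment effects that draws airborne grass blades or motes as animated sprites.
# World position offset combines sprite expansion, looping travel, sine drift and axis rotation, with optional curl noise and wind-gust push.
# Time-of-day colour with hue shift drives emissive, optionally modulated by pulsing and view-dependent fake sun shadowing.
# Opacity uses a rotating sub-UV texture or sphere mask, faded near camera and per instance.
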Material /Game/Athena/Effects/Materials/Environment/Apollo/M_Apollo_FlyingGrass
graph {
 properties {
  bUsedWithInstancedStaticMeshes=true
  bUsedWithNanite=true
  blendMode=BLEND_Translucent
  opacityMaskClipValue=0.1
  pixelDepthOffsetMode=PDOM_Legacy
  twoSided=true
  usedShadingModels=Default Lit
 }
 nodes [
  ActorPositionWS()
  Add()
  CollectionParameter() {
   collection=&FortniteMaterialParameters
   parameterName=SunAndMoonAtmosphereDirectionalVector
  }
  Constant3Vector() {constant=<rgba 0 1 1 0>}
  Constant() {r=1}
  Constant()
  n1 = NamedRerouteUsage()
  n2 = ScalarParameter() {defaultValue=1, parameterName=Glow}
  n3 = ScalarParameter() {parameterName=HueShiftAmount}
  n4 = VertexColor()
  n5 = ScalarParameter() {parameterName=HueShiftAmount}
  n6 = ScalarParameter() {parameterName=HueShiftAmountMax}
  n7 = PerInstanceRandom()
  n8 = VectorParameter() {parameterName="Color_Morning (V3)", defaultValue=<rgba 1 1 1 0>}
  n9 = VectorParameter() {parameterName="Color_Day (V3)", defaultValue=<rgba 1 1 1 0>}
  n10 = VectorParameter() {parameterName="Color_Evening (V3)", defaultValue=<rgba 1 1 1 0>}
  n11 = VectorParameter() {parameterName="Color_Night (V3)", defaultValue=<rgba 1 1 1 0>}
  n12 = MaterialFunctionCall(
   Hue Shift Percentage=StaticSwitchParameter(True=Multiply(A=@n3, B=@n4.R) {constB=-0.05}, False=LinearInterpolate(A=@n5, B=@n6, Alpha=@n7)) {defaultValue=true, parameterName=UseVertexForHueShift?}
   Texture=StaticSwitchParameter(
    True=MaterialFunctionCall() {materialFunction=/Game/Effects/Fort_Effects/Materials/_Functions/CurveColor_Time_of_Day}.Result
    False=MaterialFunctionCall(Color_Morning=@n8.RGB, Color_Day=@n9.RGB, Color_Evening=@n10.RGB, Color_Night=@n11.RGB) {materialFunction=&Color_Time_of_Day}.Result
   ) {parameterName=UseCurveforTODinsteadofScalar?}
  ) {materialFunction=&HueShift}
  n13 = Multiply(A=StaticSwitchParameter(True=Multiply(A=@n1, B=@n2), False=@n2) {parameterName=UsePulseEmissive}, B=@n12.Result)
  n14 = NamedRerouteUsage()
  MaterialFunctionCall(Emissive=StaticSwitchParameter(True=Multiply(A=@n13, B=@n14), False=@n13) {parameterName=UseFakeSunShadowing?}) {materialFunction=&MF_FX_EmissiveAdjustments}
  Multiply()
  n15 = ComponentMask(None=ObjectPositionWS()) {b=true}
  n16 = PerInstanceRandom()
  n17 = ScalarParameter() {defaultValue=1, parameterName=PulseSpeed}
  n18 = ScalarParameter() {defaultValue=1, parameterName=PulseSpeedSlow}
  n19 = PerInstanceRandom()
  n20 = Multiply(A=Time(), B=Multiply(A=LinearInterpolate(Alpha=@n16) {constA=0.5}, B=LinearInterpolate(A=@n17, B=@n18, Alpha=@n19) {constA=0.5}))
  n21 = Constant()
  n22 = Constant() {r=1}
  n23 = Saturate(None=Power(Base=MaterialFunctionCall!(Sine Phase=@n20, Value 1=@n21, Value 2=@n22) {materialFunction=/Engine/Functions/Engine_MaterialFunctions02/Utility/Sine_Remapped}.Remapped Sine) {constExponent=10})
  Multiply(A=Frac(None=Add(B=@n15) {constA=0.1}), B=@n23)
  n24 = NamedRerouteUsage()
  n25 = NamedRerouteUsage()
  n26 = TextureCoordinate()
  n27 = Constant2Vector() {g=0.5, r=0.5}
  n28 = ScalarParameter() {defaultValue=0.5, parameterName=Sprites_Radius}
  n29 = ScalarParameter() {parameterName=Sprite_Hardness}
  n30 = ComponentMask(None=TextureCoordinate()) {g=true}
  n31 = ScalarParameter() {defaultValue=2, parameterName="U_Images (S)"}
  n32 = Ceil(None=Multiply(A=VertexColor().A) {constB=4})
  n33 = TextureCoordinate()
  n34 = VectorParameter() {parameterName="Rotation Center (V2)", defaultValue=<rgba 0.5 0.5 0 0>}
  n35 = LinearInterpolate(Alpha=VertexColor().A) {constA=0.4, constB=0.6}
  n36 = MaterialFunctionCall(Speed=@n35) {materialFunction=&DebugTimeSine}
  n37 = LinearInterpolate(Alpha=@n36.Result) {constA=0.4, constB=0.6}
  n38 = Multiply(A=@n37, B=ScalarParameter() {defaultValue=4, parameterName=CenterRotation_Speed_High})
  n39 = Multiply(A=@n37, B=ScalarParameter() {defaultValue=4, parameterName=CenterRotation_Speed_Low})
  n40 = ObjectPositionWS()
  n41 = Multiply(A=Frac(None=ComponentMask(None=@n40) {b=true}), B=VertexColor().A)
  n42 = Multiply(A=Add(A=Time(), B=LinearInterpolate(A=@n38, B=@n39, Alpha=@n41)), B=ScalarParameter() {defaultValue=1, parameterName=RotationSpeed})
  n43 = TextureSampleParameter2D(UVs=MaterialFunctionCall!(
    U_Images=@n31
    V_Images=@n31
    SubUV Texture RGB
    Time Speed
    Manual Time Control=StaticBool() {value=true}
    Time=Divide(A=@n32) {constB=4}
    UVs=MaterialFunctionCall(UVs=@n33, Rotation Center=@n34.RGB, "Rotation Angle (0-1)"=@n42) {materialFunction=&CustomRotator}.Rotated Values
   ) {materialFunction=/Game/Effects/Fort_Effects/Materials/_Functions/MF_SubUV}.UVs) {parameterName=Texture, samplerType=SAMPLERTYPE_Grayscale, texture=/Game/Effects/Fort_Effects/Textures/Environments/Apollo/T_Apollo_Grass01}
  n44 = StaticSwitchParameter(True=@n43.G, False=Constant() {r=1}) {parameterName=UseFlipbookTextureAtAll?}
  n45 = Multiply(
   A=StaticSwitchParameter(True=SphereMask(A=@n26, B=@n27, Radius=@n28, Hardness=@n29), False=Multiply(A=@n30, B=@n44)) {parameterName=IsGrassMotes?}
   B=ScalarParameter() {defaultValue=1, parameterName=OpacitytMult}
  )
  n46 = MaterialFunctionCall(Distance=ScalarParameter() {defaultValue=500, parameterName=NearCamShrinkDist}) {materialFunction=&Near_Camera_Fade}
  n47 = Multiply(A=StaticSwitchParameter(True=Add(A=@n25, B=@n45), False=@n45) {parameterName=UseAddExtraDotInCenter?}, B=@n46.Result)
  n48 = Time() {bOverride_Period=true, period=3}
  n49 = Time() {bOverride_Period=true, period=8}
  n50 = ComponentMask(None=TextureCoordinate() {coordinateIndex=1}) {g=true}
  n51 = Add(
   A=StaticSwitchParameter(True=Divide(A=@n48) {constB=3}, False=Divide(A=@n49) {constB=8}) {defaultValue=true, parameterName=UseFastSprites}
   B=Multiply(A=Subtract(A=@n50) {constB=0.5}) {constB=-2}
  )
  n52 = Add(A=@n51) {constB=-0.25}
  n53 = Power(Base=ConstantBiasScale(None=Sine(None=@n52))) {constExponent=1}
  n54 = LinearInterpolate(A=Multiply(A=@n47), B=Multiply(A=@n47, B=@n53), Alpha=ScalarParameter() {parameterName=UseExtraFade?})
  n55 = Multiply(A=PerInstanceFadeAmount(), B=StaticSwitchParameter(True=Multiply(A=@n24, B=@n54), False=@n54) {parameterName=UseFakeSunShadowing?})
  n56 = Saturate(None=@n55)
  n57 = MaterialFunctionCall!(FadeOutTIme, ResetFadeInTime, FadeInTime) {materialFunction=/Game/Environments/TechArt/WindGustSystem/MF_WindGustsSystem}
  n58 = Multiply(A=@n57.ResetFadeIn, B=@n57.ResetFadeIn)
  Multiply(A=NamedRerouteUsage(), B=StaticSwitchParameter(True=Multiply(A=@n56, B=@n58), False=@n55) {parameterName=UseWindSystem?})
  Multiply()
  n59 = ScalarParameter() {defaultValue=0.5, group=Texture, parameterName=SunSide_Mult}
  n60 = ScalarParameter() {defaultValue=0.35, group=Texture, parameterName=ShadowSide_Mult}
  n61 = CollectionParameter() {
   collection=&FortniteMaterialParameters
   parameterName=SunAndMoonModelDirectionalVector
  }
  n62 = Clamp(None=DotProduct(A=ComponentMask(None=@n61) {b=true, g=true, r=true}, B=CameraVectorWS()))
  NamedRerouteDeclaration(None=Power(Base=LinearInterpolate(A=@n59, B=@n60, Alpha=@n62), Exp=ScalarParameter() {defaultValue=2, parameterName=FakeSunShadowingPower})) {name=FakeSunShadowing, nodeColor=<rgba 1 0.3866 0 1>}
  NamedRerouteDeclaration(None=LinearInterpolate(Alpha=@n23) {constA=0.25, constB=2}) {name=Pulse, nodeColor=<rgba 0.2422 0 1 1>}
  n63 = TextureCoordinate()
  n64 = Constant2Vector() {g=0.5, r=0.5}
  n65 = ScalarParameter() {defaultValue=0.5, parameterName=CenterDot_Radius}
  n66 = ScalarParameter() {parameterName=CenterDot_Hardness}
  NamedRerouteDeclaration(None=Multiply(A=SphereMask(A=@n63, B=@n64, Radius=@n65, Hardness=@n66), B=ScalarParameter() {defaultValue=1, parameterName=CenterDot_OpacitytMult})) {name=Centerdot, nodeColor=<rgba 0 0.2734 1 1>}
  n67 = ScalarParameter() {defaultValue=50, parameterName=CurlIntensity}
  n68 = ScalarParameter() {defaultValue=100, parameterName=CurlIntensityHigh}
  n69 = VertexColor()
  n70 = WorldPosition()
  n71 = ActorPositionWS()
  n72 = Time()
  n73 = ScalarParameter() {defaultValue=1, parameterName=CurlSpeed}
  n74 = Divide(
   A=Subtract(A=Subtract(A=@n70.XYZ, B=@n71), B=Multiply(A=@n72, B=@n73))
   B=LinearInterpolate(
    A=ScalarParameter() {defaultValue=1000, parameterName=CurlScale}
    B=ScalarParameter() {defaultValue=500, parameterName=CurlScale_Low}
    Alpha=@n69.A
   )
  )
  NamedRerouteDeclaration(None=Multiply(A=LinearInterpolate(A=@n67, B=@n68, Alpha=@n69.A), B=VectorNoise(World Position=@n74) {noiseFunction=VNF_VectorALU})) {name=Curl, nodeColor=<rgba 1 0.6563 0 1>}
  n75 = MaterialFunctionCall(
   TextureObject=TextureObjectParameter() {parameterName=Param}
   TextureSize=VectorParameter() {parameterName="TextureSize (V3)", defaultValue=<rgba 5000 5000 5000 0>}.RGB
   WorldPosition=VectorParameter() {parameterName=WorldTextureMask_Position, defaultValue=<rgba 0 0 0 1>}.RGB
  ) {materialFunction=/Game/Weapons/WeaponSkins/Wraps/MaterialFunctions/WorldAlignedTexture}
  NamedRerouteDeclaration(None=StaticSwitchParameter(True=ComponentMask(None=@n75.XYZ Texture) {g=true}, False=Constant() {r=1}) {parameterName=UseWorldSpaceMasking?}) {name=WorldSpaceMask, nodeColor=<rgba 0.1016 0 1 1>}
  NamedRerouteUsage()
  PerInstanceRandom()
  ReflectionVectorWS()
  StaticSwitchParameter(True=Multiply(A=@n12.Result, B=NamedRerouteUsage())) {parameterName=UseFakeSunShadowing?}
  StaticSwitchParameter() {parameterName=IsGrassMotes?}
  n76 = Multiply(
   A=MaterialFunctionCall(In=@n57.LinearTimeProgressionNormalized, Power=ScalarParameter() {defaultValue=3, parameterName=WindGustsAccelaration}) {materialFunction=&SCurve}.Result
   B=ScalarParameter() {defaultValue=1500, parameterName=WindGustsDistanceOffset}
  )
  n77 = MaterialFunctionCall!(Position, IsTrees, IsGrass, IsMediumPlant, IsMediumPlant_2024Anim, IsMediumPlant2ndLevel, MediumPlant_BasePivotPostion) {materialFunction=/Game/Environments/Asteria/Foliage/Wind/MF_WindActorData}
  n78 = WorldPosition()
  n79 = Multiply(A=ScalarParameter() {defaultValue=0.25, parameterName=MinSizeMult}, B=VertexColor().G)
  n80 = LinearInterpolate(
   A=ScalarParameter() {defaultValue=500, parameterName=YmaxSize}
   B=ScalarParameter() {defaultValue=500, parameterName=YmaxSize_Low}
   Alpha=@n36.Result
  )
  n81 = ComponentMask(None=LinearInterpolate(A=Multiply(A=@n79, B=@n80), B=@n80, Alpha=@n53)) {r=true}
  n82 = LinearInterpolate(
   A=ScalarParameter() {defaultValue=500, parameterName=XmaxSize}
   B=ScalarParameter() {defaultValue=500, parameterName=XmaxSize_Low}
   Alpha=MaterialFunctionCall(Speed=Add(A=@n35) {constB=0.25}) {materialFunction=&DebugTimeSine}.Result
  )
  n83 = Multiply(A=@n79, B=@n82)
  n84 = AppendVector(A=@n81, B=ComponentMask(None=LinearInterpolate(A=@n83, B=@n82, Alpha=@n53)) {g=true})
  n85 = Multiply(
   A=NamedRerouteUsage()
   B=Multiply(A=StaticSwitchParameter(True=@n84, False=@n81) {defaultValue=true, parameterName=UseXandYSize?}, B=@n46.Result)
  )
  n86 = Multiply(A=Time(), B=ScalarParameter() {defaultValue=0.001, parameterName=VarianceSpeed}) {constB=0.05}
  n87 = ComponentMask(None=TextureCoordinate() {coordinateIndex=1}) {r=true}
  n88 = Add(A=Multiply(A=Time(), B=ScalarParameter() {defaultValue=0.4, parameterName=DriftSpeed}), B=Sine(None=Add(A=@n86, B=@n87)) {period=0.17})
  n89 = Sine(None=@n88) {period=0.56}
  n90 = Sine(None=@n88) {period=0.7}
  n91 = Add(A=@n86, B=@n87)
  n92 = Multiply(
   A=AppendVector(A=AppendVector(A=@n89, B=@n90), B=Sine(None=@n88) {period=0.82})
   B=LinearInterpolate(
    A=ScalarParameter() {defaultValue=1, parameterName=MinHorizontalDrift}
    B=ScalarParameter() {defaultValue=1, parameterName=MaxHorizontalDrift}
    Alpha=Sine(None=@n91) {period=0.14}
   )
  )
  n93 = Multiply(A=ObjectRadius(), B=Constant() {r=0.5}) {constB=0.5}
  n94 = Multiply(A=@n93) {constB=-1}
  n95 = Frac(None=@n51)
  n96 = ScalarParameter() {defaultValue=1, parameterName=OverallSpeed}
  n97 = ScalarParameter() {defaultValue=1, parameterName=OverallSpeed_Low}
  n98 = VertexColor()
  n99 = ObjectPositionWS()
  n100 = Multiply(A=@n98.A, B=Frac(None=ComponentMask(None=@n99) {b=true}))
  n101 = Multiply(
   A=Multiply(A=LinearInterpolate(A=@n94, B=@n93, Alpha=@n95), B=LinearInterpolate(A=@n96, B=@n97, Alpha=@n100))
   B=LinearInterpolate(
    A=VectorParameter() {parameterName=OverallDirection, defaultValue=<rgba 1 0 0 1>}.RGB
    B=VectorParameter() {parameterName=OverallDirection_Low, defaultValue=<rgba 1 0 0 1>}.RGB
    Alpha=@n98.A
   )
  )
  n102 = VectorParameter() {parameterName=RotationAxis}
  n103 = Time()
  n104 = ObjectPositionWS()
  n105 = StaticSwitchParameter(True=Add(A=@n103, B=ComponentMask(None=@n104) {b=true}), False=@n103) {parameterName=UsewithGrassActor?}
  n106 = ObjectPositionWS()
  n107 = LinearInterpolate(
   A=ScalarParameter() {defaultValue=1, parameterName=RotationSpeed}
   B=ScalarParameter() {defaultValue=1, parameterName=RotationSpeed_Low}
   Alpha=Frac(None=ComponentMask(None=@n106) {b=true})
  )
  n108 = ActorPositionWS()
  n109 = Add(
   A=Add(
    A=MaterialFunctionCall!(Center Location=@n78.XYZ, 0-1 UVs, XY scale=@n85, OverridePivot, "Normals (Optional)") {materialFunction=&Sprite}.World position offset
    B=Add(A=@n92, B=@n101)
   )
   B=RotateAboutAxis(
    NormalizedRotationAxis=Transform(None=@n102.RGB) {transformSourceType=TRANSFORMSOURCE_Local}
    RotationAngle=Multiply(A=@n105, B=@n107)
    PivotPoint=StaticSwitchParameter(True=@n104, False=@n108) {parameterName=UsewithGrassActor?}
    Position=WorldPosition().XYZ
   )
  )
  n110 = StaticSwitchParameter(True=Add(A=@n109, B=NamedRerouteUsage()), False=@n109) {parameterName=UseCurl?}
  StaticSwitchParameter(True=Add(A=Multiply(A=@n76, B=@n77.WindI_Direction), B=@n110), False=@n110) {parameterName=UseWindSystem?}
  StaticSwitchParameter(True=Multiply(A=@n12.Result, B=NamedRerouteUsage()), False=@n12.Result) {parameterName=UseFakeSunShadowing?}
  VertexColor()
  WorldPosition()
  WorldPosition()
 ]
}
